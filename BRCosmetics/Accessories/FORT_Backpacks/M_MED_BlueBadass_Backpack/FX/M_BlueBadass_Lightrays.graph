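# Translucent, unlit, two-sided material for animated light rays on the BlueBadass backpack cosmetic.
# A heat gradient texture passes through HDR remapping and emissive adjustment functions for color.
# Opacity multiplies two opposite-panning ray textures with scrolling noise, adds a V-based hot band, then applies a noise-varied vertical falloff.
# Progression floors against U as a hard reveal mask, and depth fade softens intersections except on mobile.
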Material /BRCosmetics/Accessories/FORT_Backpacks/M_MED_BlueBadass_Backpack/FX/M_BlueBadass_Lightrays
graph {
 properties {blendMode=BLEND_Translucent, pixelDepthOffsetMode=PDOM_Legacy, shadingModel=MSM_Unlit, twoSided=true, usedShadingModels=Unlit}
 nodes [
  n1 = TextureSample() {texture=/BRCosmetics/Weapons/WeaponSkins/Textures/BadassMale_Heat_Gradient}
  MaterialFunctionCall(Emissive=MaterialFunctionCall(LDR Input=@n1.RGB, HDR Tint=@n1.RGB) {materialFunction=&DeriveHDRfromLDR}.RemappedRange) {materialFunction=&MF_FX_EmissiveAdjustments}
  n2 = ComponentMask(None=Add(A=TextureCoordinate(), B=ScalarParameter() {defaultValue=-0.03, parameterName=Slide})) {g=true}
  OneMinus(None=@n2)
  n3 = TextureCoordinate()
  n4 = ScalarParameter() {defaultValue=2, parameterName=Gradient Tile A}
  n5 = Multiply(
   A=ScalarParameter() {defaultValue=1, parameterName=Panner Speed}
   B=Add(A=ScalarParameter() {defaultValue=1, parameterName=RandomSpeed}, B=Time())
  )
  n6 = TextureSample(UVs=Panner(Coordinate=Multiply(A=@n3, B=@n4), Time=@n5) {speedX=0.0702}) {texture=/Game/Effects/Fort_Effects/Textures/Outlander/T_InTheZone_LightRays_01}
  n7 = TextureCoordinate()
  n8 = ScalarParameter() {defaultValue=2.5, parameterName=Gradient Tile B}
  n9 = TextureSample(UVs=Panner(Coordinate=Multiply(A=@n7, B=@n8), Time=@n5) {speedX=-0.047}) {texture=&T_InTheZone_LightRays_01}
  n10 = TextureCoordinate() {uTiling=0.33, vTiling=0.01}
  n11 = ScalarParameter() {defaultValue=1, parameterName=RandomSpeed}
  n12 = TextureSample(UVs=Panner(Coordinate=Add(A=@n10, B=@n11), Time=@n5) {speedX=0.0436}) {texture=&T_NoisePack_01}
  n13 = ScalarParameter() {defaultValue=3, parameterName=Mask Brightness}
  n14 = ScalarParameter() {defaultValue=6.5, parameterName=Center Exponent}
  n15 = Add(
   A=Multiply(A=Multiply(A=@n6.R, B=@n9.G), B=Multiply(A=@n12.R, B=@n13))
   B=Multiply(A=Power(Base=@n2, Exp=@n14), B=ScalarParameter() {defaultValue=2, parameterName=Center Brightness})
  )
  n16 = Panner(Coordinate=TextureCoordinate() {uTiling=2, vTiling=0}) {speedX=0.2, speedY=0.1}
  n17 = Power(
   Base=ComponentMask(None=TextureCoordinate()) {g=true}
   Exp=LinearInterpolate(Alpha=TextureSample(UVs=@n16) {texture=&T_NoisePack_01}.R) {constA=1, constB=4}
  )
  n18 = ComponentMask(None=TextureCoordinate()) {r=true}
  n19 = ScalarParameter() {defaultValue=1, parameterName=Progression}
  n20 = Multiply(
   A=Multiply(A=Multiply(A=@n15, B=@n17), B=ScalarParameter() {defaultValue=0.65, parameterName=Alpha, sliderMin=1})
   B=Floor(None=Add(A=@n18, B=@n19))
  )
  ShadingPathSwitch(Default=DepthFade(Opacity=@n20, FadeDistance=ScalarParameter() {defaultValue=5, parameterName=Fade Distance}), Mobile=@n20) {desc=Migrated from FeatureLevelSwitch}
 ]
}
